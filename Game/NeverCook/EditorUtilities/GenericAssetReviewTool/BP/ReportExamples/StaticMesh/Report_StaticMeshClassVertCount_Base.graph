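# Editor-only base report for the Generic Asset Review Tool that flags static meshes exceeding per-quality-level vertex budgets set by class, such as building walls and props.
# It matches a Blueprint to ClassBudgets by NativeParentClass tag or child-class test, then checks its Asset Registry static mesh dependencies.
# Failures list offending levels, with color shifting toward red by overage.
# Separate loaded and tag-only MinLOD check functions exist, uncalled within this graph.
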EditorUtilityBlueprint /Game/NeverCook/EditorUtilities/GenericAssetReviewTool/BP/ReportExamples/StaticMesh/Report_StaticMeshClassVertCount_Base
graphs [
 CheckIfAssetClassAllowed {nodes [
   n1 = CheckIfAssetClassAllowed!() -> (then, AssetData:Asset Data Structure, AllowedClasses:Array of Object Soft Class References) {is=title K2Node_FunctionEntry}
   > Set l_classes(l_classes:Array of Object Soft Class References=@n1.AllowedClasses) -> (Output_Get:Array of Object Soft Class References)
   > n2 = Keys!(
    execute
    TargetMap:Map of Object Soft Class References to GART Vert Budget Struct Structures=Get ClassBudgets() -> (ClassBudgets:Map of Object Soft Class References to GART Vert Budget Struct Structures)
   ) -> (then, Keys:Array of Object Soft Class References)
   > n3 = For Each Loop(Array:Array of Object Soft Class References=@n2) -> (Array Element:Object Soft Class Reference)
   n4 = Parent: CheckIfAssetClassAllowed!(
    execute=@n3.Completed
    AssetData:Asset Data Structure=@n1
    AllowedClasses:Array of Object Soft Class References=Get l_classes() -> (l_classes:Array of Object Soft Class References)
   ) -> (then, Allowed:bool) {is=title K2Node_CallParentFunction}
   > Return Node!(execute, Allowed:bool=@n4)
   Add Unique(execute=@n3.LoopBody, TargetArray:Array of Object Soft Class References=Get l_classes(), NewItem:"Object Soft Class Reference (by ref)"=@n3)
  ]}
 EventGraph {links [], nodes []}
 FindAssetClassBudget {nodes [
   FindAssetClassBudget!() -> (then, InAssetData:"Asset Data Structure (by ref)") {is=title K2Node_FunctionEntry}
   n1 = GetTagValue(InAssetData=Get InAssetData() -> (InAssetData:"Asset Data Structure (by ref)"), InTagName=MakeLiteralName(Value=NativeParentClass))
   > Set l_inAssetSoftClass(l_inAssetSoftClass:Object Soft Class Reference=To Soft Class Reference(SoftClassPath=MakeSoftClassPath(PathString=@n1))) -> (Output_Get:Object Soft Class Reference)
   n2 = Get ClassBudgets()
   > n3 = Values!(execute, TargetMap:Map of Object Soft Class References to GART Vert Budget Struct Structures=@n2) -> (then, Values:Array of GART Vert Budget Struct Structures)
   > Set l_values(l_values:Array of GART Vert Budget Struct Structures=@n3) -> (Output_Get:Array of GART Vert Budget Struct Structures)
   > n4 = Keys(TargetMap=@n2)
   > n5 = For Each Loop with Break!(Exec:Exec, Array:Array of Object Soft Class References=@n4, Break:Exec=@n7.then) -> (LoopBody:Exec, Array Element:Object Soft Class Reference, Array Index:int, Completed:Exec)
   Return Node!(
    execute=@n5.Completed
    FoundBudget:bool=Get l_foundBudget() -> (l_foundBudget:bool)
    BudgetData:GART Vert Budget Struct Structure=Get l_finalBudget() -> (l_finalBudget:GART Vert Budget Struct Structure)
    TestClass:Object Soft Class Reference=Get l_foundTestClass() -> (l_foundTestClass:Object Soft Class Reference)
   )
   > Set l_index(l_index:int=@n5.Array Index) -> (Output_Get:int=0)
   > Set l_foundTestClass(l_foundTestClass:Object Soft Class Reference=@n5) -> (Output_Get:Object Soft Class Reference)
   > n6 = Branch(Condition="Equal (SoftClassReference)"(
     A:"Object Soft Class Reference (by ref)"=@n5
     B:"Object Soft Class Reference (by ref)"=Get l_inAssetSoftClass() -> (l_inAssetSoftClass:Object Soft Class Reference)
    ) -> (ReturnValue:bool=~) {is=title K2Node_PromotableOperator})
   Set l_finalBudget(
    execute=[@n6.then, @n12.then, @n14.then]
    l_finalBudget:GART Vert Budget Struct Structure="Get (a copy)"(
     Array:Array of GART Vert Budget Struct Structures=Get l_values() -> (l_values:Array of GART Vert Budget Struct Structures)
     Dimension 1=Get l_index() -> (l_index:int=0)
    ) -> (Output:GART Vert Budget Struct Structure)
   ) -> (Output_Get:GART Vert Budget Struct Structure)
   > n7 = Set l_foundBudget(l_foundBudget:bool=true) -> (Output_Get:bool)
   n8 = LoadClassAsset_Blocking(execute=@n6.else, AssetClass=@n5)
   > Set l_testClass(l_testClass:Object Class Reference=@n8) -> (Output_Get:Object Class Reference)
   > n9 = Branch(Condition=IsAssetLoaded!(InAssetData:"Asset Data Structure (by ref)"=Get InAssetData()) -> (ReturnValue:bool))
   n10 = GetBlueprintAsset(execute=[@n14.else, @n9.then], Object=GetAsset(InAssetData=Get InAssetData()))
   > n11 = GeneratedClass!(execute, BlueprintObj:Blueprint Object Reference=@n10) -> (then, ReturnValue:Object Class Reference)
   > n12 = Branch(Condition=ClassIsChildOf(TestClass=@n11, ParentClass=Get l_testClass() -> (l_testClass:Object Class Reference)))
   n13 = LoadClassAsset_Blocking(execute=@n9.else, AssetClass=Get l_inAssetSoftClass())
   > n14 = Branch(Condition=ClassIsChildOf(TestClass=@n13, ParentClass=Get l_testClass()))
  ]}
 Load_MinLODCheck {nodes [
   n1 = Load_MinLODCheck!() -> (then, AssetData:Asset Data Structure, l_budget1:Map of EPerQualityLevels Enums to Integers=Low) {is=title K2Node_FunctionEntry}
   > n2 = Set l_assetData(l_assetData=@n1)
   > Set l_budget(l_budget:Map of EPerQualityLevels Enums to Integers=@n1.l_budget1) -> (Output_Get:Map of EPerQualityLevels Enums to Integers=Low)
   > n3 = Cast To StaticMesh(Object=GetAsset(InAssetData=@n2))
   > Set l_staticMesh(l_staticMesh:Static Mesh Object Reference=@n3) -> (Output_Get:Static Mesh Object Reference)
   > n4 = GetEditorWorld(self=Get UnrealEditorSubsystem())
   > Set l_world(l_world:Object Reference=@n4) -> (Output_Get:Object Reference)
   > Set l_valid(l_valid:bool=true) -> (Output_Get:bool)
   > n5 = Sequence!(execute) -> (then_0, then_1, then_2:Exec, then_3:Exec)
   Set l_finalLongDesc(execute=@n5.then_1, l_finalLongDesc:String=Append(B=Missing required MinLOD Quality Levels:)) -> (Output_Get:String)
   n6 = GetMinLODForQualityLevels!(self:Static Mesh Object Reference=Get l_staticMesh() -> (l_staticMesh:Static Mesh Object Reference)) -> (QualityLevelMinimumLODs:Map of EPerQualityLevels Enums to Integers=Low, Default:int=0)
   > n7 = Set l_meshQualityLevels(l_meshQualityLevels:Map of EPerQualityLevels Enums to Integers=@n6) -> (Output_Get:Map of EPerQualityLevels Enums to Integers=Low)
   > Add!(
    execute
    TargetMap:Map of EPerQualityLevels Enums to Integers=@n7
    Key:"EPerQualityLevels Enum (by ref)"=Num
    Value:"Integer (by ref)"=@n6.Default
   ) -> (then) {is=title K2Node_CallFunction}
   > n8 = LoadAssetAndGetVertexCounts!(execute, AssetData:"Asset Data Structure (by ref)"=Get l_assetData(), __WorldContext:Object Reference=Get l_world()) -> (then, VertsPerLOD:Array of Integers)
   > Set l_vertsPerLOD(l_vertsPerLOD:Array of Integers=@n8) -> (Output_Get:Array of Integers)
   n9 = Get l_budget() -> (l_budget:Map of EPerQualityLevels Enums to Integers=Low)
   n10 = Keys(execute=@n5.then_2, TargetMap:Map of EPerQualityLevels Enums to Integers=@n9) -> (Keys:Array of EPerQualityLevels Enums)
   > n11 = Values(TargetMap:Map of EPerQualityLevels Enums to Integers=@n9) -> (Values:Array of Integers)
   > n12 = For Each Loop(Array:Array of EPerQualityLevels Enums=@n10) -> (Array Element:EPerQualityLevels Enum)
   > Set l_curQualityToCheck(l_curQualityToCheck:EPerQualityLevels Enum=@n12) -> (Output_Get:EPerQualityLevels Enum=Low)
   > Set l_vertCountMax(l_vertCountMax:int="Get (a ref)"(Array=@n11, Dimension 1=@n12.Array Index)) -> (Output_Get:int=0)
   n13 = Find(
    TargetMap:Map of EPerQualityLevels Enums to Integers=Get l_meshQualityLevels() -> (l_meshQualityLevels:Map of EPerQualityLevels Enums to Integers=Low)
    Key:"EPerQualityLevels Enum (by ref)"=@n12
   ) -> (Value:int)
   > n14 = Branch(Condition=@n13.ReturnValue)
   > Set l_qualityLevelIndex(l_qualityLevelIndex:int=@n13) -> (Output_Get:int=0)
   > n15 = Branch(Condition=Is Valid Index!(
     TargetArray:Array of Integers=Get l_vertsPerLOD() -> (l_vertsPerLOD:Array of Integers)
     IndexToTest:int=Get l_qualityLevelIndex() -> (l_qualityLevelIndex:int=0)
    ) -> (ReturnValue:bool))
   > Set l_vertCount(l_vertCount:int="Get (a ref)"(Array=Get l_vertsPerLOD(), Dimension 1=Get l_qualityLevelIndex() -> (l_qualityLevelIndex=0))) -> (Output_Get:int=0)
   > Branch(Condition="integer > integer"(A=Get l_vertCount() -> (l_vertCount:int=0), B=Get l_vertCountMax() -> (l_vertCountMax:int=0)))
   n16 = Get l_vertCount() -> (l_vertCount=0)
   n17 = Get l_vertCountMax() -> (l_vertCountMax=0)
   n18 = Get l_curQualityToCheck() -> (l_curQualityToCheck:EPerQualityLevels Enum=Low)
   > Set l_finalLongDesc(l_finalLongDesc=Append!(
     A=Get l_finalLongDesc() -> (l_finalLongDesc:String)
     B="\r\nLOD Used on "
     C:String=Enum to String!(Enumerator:EPerQualityLevels Enum=@n18) -> (ReturnValue:String)
     D:String=" is "
     E:String="To String (Integer)"(InInt=@n16)
     F:String=" Verts and should be less than: "
     G:String="To String (Integer)"(InInt=@n17)
    ) -> (ReturnValue))
   > Set l_valid(l_valid=false)
   n19 = Get l_qualityLevelIndex() -> (l_qualityLevelIndex=0)
   n20 = Length(TargetArray:Array of Integers=Get l_vertsPerLOD())
   n21 = Get l_curQualityToCheck() -> (l_curQualityToCheck=Low)
   Set l_finalLongDesc(
    execute=@n15.else
    l_finalLongDesc=Append!(
     A=Get l_finalLongDesc()
     B="\r\nMinLOD for "
     C=Enum to String(Enumerator=@n21)
     D=" is "
     E="To String (Integer)"(InInt=@n19)
     F=" but the mesh only has "
     G="To String (Integer)"(InInt=@n20)
     H:String=" LODs.  Reduce the index or add LODs so it's valid."
    ) -> (ReturnValue)
   )
   > Set l_valid(l_valid=false)
   Set l_vertCount(execute=@n14.else, l_vertCount="Get (a ref)"(Array=Get l_vertsPerLOD(), Dimension 1=0)) -> (Output_Get=0)
   > Branch(Condition="integer > integer"(A=Get l_vertCount() -> (l_vertCount=0), B=Get l_vertCountMax() -> (l_vertCountMax=0)))
   n22 = Get l_vertCount() -> (l_vertCount=0)
   n23 = Get l_vertCountMax() -> (l_vertCountMax=0)
   n24 = Get l_curQualityToCheck() -> (l_curQualityToCheck=Low)
   > Set l_finalLongDesc(l_finalLongDesc=Append!(
     A=Get l_finalLongDesc()
     B="\r\nNo MinLOD is assigned to "
     C=Enum to String(Enumerator=@n24)
     D=" and LOD0 is too many verts: "
     E="To String (Integer)"(InInt=@n22)
     F=" and should be less than: "
     G="To String (Integer)"(InInt=@n23)
    ) -> (ReturnValue))
   > Set l_valid(l_valid=false)
   Return Node!(
    execute=@n5.then_3
    Valid:bool=Get l_valid() -> (l_valid:bool)
    ReportData:Struct GART Report Data Structure=Make Struct GART Report Data(
     DisplayTitle_20_#1=LOD
     DisplayDescShort_29_#2=MinLOD Vert Count
     DisplayDescLong_30_#3=Get l_finalLongDesc()
    )
   )
  ]}
 NoLoad_MinLODCheck {nodes [
   n1 = NoLoad_MinLODCheck!() -> (then, AssetData:Asset Data Structure) {is=title K2Node_FunctionEntry}
   > Set l_assetData(l_assetData=@n1)
   > n2 = GetEditorWorld(self=Get UnrealEditorSubsystem())
   > Set l_world(l_world=@n2)
   > Set l_valid(l_valid=true)
   > n3 = Sequence!(execute) -> (then_0, then_1, then_2)
   n4 = Break Asset Data(AssetData=Get l_assetData())
   > Set l_finalLongDesc(l_finalLongDesc=Append!(A="To String (Name)"(InName=@n4.AssetName), B=" is missing required MinLOD Quality Levels:") -> (ReturnValue))
   n5 = Get l_world()
   > n6 = NoLoad_GetMinLODLevelFromAssetData!(execute, AssetData:Asset Data Structure=Get l_assetData(), __WorldContext:Object Reference=@n5) -> (then, QualitiesInUse:Map of EPerQualityLevels Enums to Integers=Low)
   > Set l_meshQualityLevel(l_meshQualityLevel:Map of EPerQualityLevels Enums to Integers=@n6) -> (Output_Get:Map of EPerQualityLevels Enums to Integers=Low)
   > n7 = NoLoad_GetStaticMeshVertCountFromAssetData!(execute, AssetData:Asset Data Structure=Get l_assetData(), __WorldContext:Object Reference=@n5) -> (then, VertCount:int=0)
   > Set l_vertCount(l_vertCount=@n7) -> (Output_Get=0)
   n8 = Get QualityLevelMaxVerts() -> (QualityLevelMaxVerts:Map of EPerQualityLevels Enums to Integers=Low)
   n9 = Keys(execute=@n3.then_1, TargetMap=@n8)
   > n10 = Values(TargetMap=@n8)
   > n11 = For Each Loop(Array=@n9)
   > n12 = Set l_vertCountMax(l_vertCountMax="Get (a ref)"(Array=@n10, Dimension 1=@n11.Array Index)) -> (Output_Get=0)
   > Branch(Condition="integer > integer"(A=Get l_vertCount() -> (l_vertCount=0), B=@n12))
   n13 = Find(TargetMap=Get l_meshQualityLevel() -> (l_meshQualityLevel:Map of EPerQualityLevels Enums to Integers=Low), Key=@n11)
   > n14 = Branch(Condition=@n13.ReturnValue)
   > n15 = Branch(Condition="integer > integer"(A=@n13))
   n16 = Get l_vertCount() -> (l_vertCount=0)
   n17 = Get l_vertCountMax() -> (l_vertCountMax=0)
   n18 = Set l_finalLongDesc(
    execute=@n15.else
    l_finalLongDesc=Append!(
     A=Get l_finalLongDesc()
     B="\r\nLOD Used on "
     C:String=Enum to String(Enumerator=@n11)
     D:String=" is "
     E:String="To String (Integer)"(InInt=@n16)
     F:String=" Verts and should be less than: "
     G:String="To String (Integer)"(InInt=@n17)
    ) -> (ReturnValue)
   )
   Set l_valid(execute=[@n18.then, @n21.then], l_valid=false)
   n19 = Get l_vertCount() -> (l_vertCount=0)
   n20 = Get l_vertCountMax() -> (l_vertCountMax=0)
   n21 = Set l_finalLongDesc(
    execute=@n14.else
    l_finalLongDesc=Append(
     A=Get l_finalLongDesc()
     B="\r\nNo MinLOD is assigned to "
     C=Enum to String(Enumerator=@n11)
     D=" and LOD0 is too many verts: "
     E="To String (Integer)"(InInt=@n19)
     F=" and should be less than: "
     G="To String (Integer)"(InInt=@n20)
    )
   )
   Return Node(
    execute=@n3.then_2
    Valid=Get l_valid()
    ReportData=Make Struct GART Report Data(
     DisplayTitle_20_#1=LOD
     DisplayDescShort_29_#2=MinLOD Vert Count
     DisplayDescLong_30_#3=Get l_finalLongDesc()
    )
   )
  ]}
 RunFilterLogicOnAssetData {nodes [
   n1 = RunFilterLogicOnAssetData()
   n2 = GetBlueprintAsset(Object=GetAsset(InAssetData=@n1))
   > n3 = Is Valid(InputObject=@n2)
   n4 = Sequence!(execute=[@n1.then, @n3.Is Valid]) -> (then_0, then_1)
   > n5 = FindAssetClassBudget!(execute, self, InAssetData:"Asset Data Structure (by ref)"=@n1) -> (then, FoundBudget:bool, BudgetData:GART Vert Budget Struct Structure, TestClass:Object Soft Class Reference) {is=title K2Node_CallFunction}
   > Set l_meshQualityLevel(l_meshQualityLevel=Break GART Vert Budget Struct!(GART_VertBudgetStruct:"GART Vert Budget Struct Structure (by ref)"=@n5.BudgetData) -> (VertsPerQualityLevel_7_#4:Map of EPerQualityLevels Enums to Integers) {is=title K2Node_BreakStruct}) -> (Output_Get=Low)
   > Set l_foundClass(l_foundClass:Object Soft Class Reference=@n5.TestClass) -> (Output_Get:Object Soft Class Reference)
   > n6 = Branch(Condition=@n5)
   > n7 = Sequence()
   > n8 = Get Dependencies!(
    execute
    self:Asset Registry Interface=GetAssetRegistry()
    PackageName:Name=Break Asset Data(AssetData=Get AssetData())
    DependencyOptions:"Asset Registry Dependency Options Structure (by ref)"=Make Asset Registry Dependency Options!(bIncludeGamePackageReferences, bIncludeEditorOnlyPackageReferences) -> (AssetRegistryDependencyOptions)
   ) -> (then, OutDependencies:Array of Names, ReturnValue:bool) {is=title K2Node_CallFunction}
   > Branch(Condition=Is Valid Index(TargetArray:Array of Names=@n8))
   > n9 = GetAssets!(
    execute
    self:Asset Registry Interface=GetAssetRegistry()
    Filter:"ARFilter Structure (by ref)"=MakeARFilter!(
     PackageNames:Array of Names=@n8
     PackagePaths:Array of Names
     SoftObjectPaths:Array of Soft Object Path Structures
     ClassPaths:Array of Top Level Asset Path Structures=Make Array!("[0]_PackageName":String=/Script/Engine, "[0]_AssetName":String=StaticMesh) -> (Array:Array of Top Level Asset Path Structures)
     RecursiveClassPathsExclusionSet:Set of Top Level Asset Path Structures
     ClassNames:Array of Names
     RecursiveClassesExclusionSet:Set of Names
     bRecursivePaths:bool=false
     bRecursiveClasses:bool=true
     bIncludeOnlyOnDiskAssets:bool=false
    ) -> (ReturnValue:ARFilter Structure)
    bSkipARFilteredAssets:bool=true
   ) -> (then, OutAssetData:Array of Asset Data Structures, ReturnValue:bool)
   > Set l_meshDependencies(l_meshDependencies:Array of Asset Data Structures=@n9) -> (Output_Get:Array of Asset Data Structures)
   n10 = For Each Loop(Exec=@n7.then_1, Array:Array of Asset Data Structures=Get l_meshDependencies() -> (l_meshDependencies:Array of Asset Data Structures)) -> (Array Element:Asset Data Structure)
   > n11 = Cast To StaticMesh(Object=GetAsset(InAssetData=@n10))
   > n12 = IsMeshWithinVertBudget!(
    execute
    mesh:Static Mesh Object Reference=@n11
    QualityBudgets:Map of EPerQualityLevels Enums to Integers=Get l_meshQualityLevel() -> (l_meshQualityLevel=Low)
    __WorldContext:Object Reference=Get l_world()
   ) -> (then, BadQualityMap:Map of EPerQualityLevels Enums to Vector 2D Structures=Low)
   > Branch(Condition="integer > integer"(A=Length!(TargetMap:Map of EPerQualityLevels Enums to Vector 2D Structures=@n12) -> (ReturnValue:int=0) {is=title K2Node_CallFunction}))
   n13 = Get l_foundClass() -> (l_foundClass:Object Soft Class Reference)
   > AddError(NewError=Append!(
     A="The static mesh "
     B=GetDisplayName(Object=@n11)
     C=" exceeds vertex budget for the following quality levels for something used as a "
     D=To Soft Class Path!(SoftClassReference:Object Soft Class Reference=@n13) -> (ReturnValue_PathString:String)
    ) -> (ReturnValue))
   > n14 = Values(TargetMap:Map of EPerQualityLevels Enums to Vector 2D Structures=@n12) -> (Values:Array of Vector 2D Structures)
   > Set l_values(l_values:Array of Vector 2D Structures=@n14) -> (Output_Get:Array of Vector 2D Structures)
   > n15 = Keys(TargetMap:Map of EPerQualityLevels Enums to Vector 2D Structures=@n12)
   > n16 = For Each Loop(Array:Array of EPerQualityLevels Enums=@n15) -> (Array Element:EPerQualityLevels Enum)
   n17 = BreakVector2D(InVec="Get (a ref)"(Array:Array of Vector 2D Structures=Get l_values() -> (l_values:Array of Vector 2D Structures), Dimension 1=@n16.Array Index) -> (Output:"Vector 2D Structure (by ref)"))
   > AddError(
    NewError=Append!(
     A="Quality level "
     B=Enum to String(Enumerator=@n16)
     C=" is "
     D="To String (Float)"(InDouble=@n17)
     E:String=" verts and should be <= "
     F:String="To String (Float)"(InDouble=@n17.Y)
    ) -> (ReturnValue)
    Depth=1
   )
   n18 = float / float(A=@n17, B=@n17.Y)
   > Set l_colorLerpValue(l_colorLerpValue:double="Max (Float)"(
     A=Get l_colorLerpValue() -> (l_colorLerpValue:double)
     B=NormalizeToRange!(Value:double=@n18, RangeMin:double=1, RangeMax:double=4) -> (ReturnValue:double)
    )) -> (Output_Get:double)
   Return Node(execute=@n6.else)
   n19 = Get l_colorLerpValue()
   Return Node(
    execute=@n4.then_1
    Valid=IsEmpty(InString=GetCurrentErrorString!() -> (CurrentErrorString:String))
    ReportData=Make Struct GART Report Data(
     DisplayTitle_20_#1=Vertex Count
     DisplayDescShort_29_#2=Too many verts for quality levels.
     DisplayColorOverride_40_#5="Lerp (LinearColor)"(A=(R=0.2794,G=0.3438,B=0,A=1), B=(R=1,G=0,B=0,A=1), Alpha=@n19)
    )
   )
  ]}
]
variables [
 ClassBudgets {default {
   &BuildingFloor {vertsPerQualityLevel {high=1500, low=750, medium=1000}}
   &BuildingProp {vertsPerQualityLevel {high=7000, low=1000, medium=2000}}
   &BuildingRoof {vertsPerQualityLevel {high=6000, low=3000, medium=4500}}
   &BuildingStairs {vertsPerQualityLevel {high=8000, low=4000, medium=6500}}
   &BuildingWall {vertsPerQualityLevel {high=4000, low=1500, medium=2500}}
  }}
 QualityLevelMaxVerts {default {high=15000, low=1500, medium=5000}}
]
