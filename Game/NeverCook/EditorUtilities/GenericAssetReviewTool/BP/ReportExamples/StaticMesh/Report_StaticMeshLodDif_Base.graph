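# Editor-only Generic Asset Review Tool example report that checks whether each static mesh LOD cuts enough vertices from the previous one.
# When loading is allowed, Load_MinLODCheck gets per-LOD vertex counts and logs errors for non-decreasing counts or reductions below PercentDiffRequiredBetweenLevels.
# It returns valid only without errors, with a report color lerped from the worst tracked under-reduction ratio.
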EditorUtilityBlueprint /Game/NeverCook/EditorUtilities/GenericAssetReviewTool/BP/ReportExamples/StaticMesh/Report_StaticMeshLodDif_Base
graphs [
 EventGraph {links [], nodes []}
 Load_MinLODCheck {nodes [
   n1 = Load_MinLODCheck!() -> (then, AssetData:Asset Data Structure) {is=title K2Node_FunctionEntry}
   > n2 = Set l_assetData(l_assetData=@n1)
   > n3 = Cast To StaticMesh(Object=GetAsset(InAssetData=@n2))
   > Set l_staticMesh(l_staticMesh:Static Mesh Object Reference=@n3) -> (Output_Get:Static Mesh Object Reference)
   > n4 = Sequence!(execute) -> (then_0, then_1, then_2:Exec)
   > n5 = LoadAssetAndGetVertexCounts!(execute, AssetData:"Asset Data Structure (by ref)"=Get l_assetData(), __WorldContext:Object Reference=Get l_world()) -> (then, VertsPerLOD:Array of Integers)
   > Set l_vertsPerLOD(l_vertsPerLOD:Array of Integers=@n5) -> (Output_Get:Array of Integers)
   n6 = For Each Loop(Exec=@n4.then_1, Array:Array of Integers=Get l_vertsPerLOD() -> (l_vertsPerLOD:Array of Integers)) -> (Array Element:int)
   > Set l_lodIndex(l_lodIndex:int=@n6.Array Index) -> (Output_Get:int=0)
   > Set l_lodVertCount(l_lodVertCount:int=@n6) -> (Output_Get:int=0)
   > n7 = Sequence!(execute) -> (then_0, then_1)
   n8 = Get l_lodVertCount() -> (l_lodVertCount:int=0)
   n9 = Get l_lastCount() -> (l_lastCount:int=0)
   > PrintString(InString=Append!(A="To String (Integer)"(InInt=@n8), B=" >? ", C:String="To String (Integer)"(InInt=@n9)) -> (ReturnValue))
   > n10 = Branch(Condition="integer >= integer"(A=Get l_lodVertCount() -> (l_lodVertCount=0), B=Get l_lastCount() -> (l_lastCount=0)))
   n11 = Get l_lodIndex() -> (l_lodIndex:int=0)
   > AddError(NewError=Append(A="LOD ", B="To String (Integer)"(InInt=@n11), C=" has more/same verts than the previous LOD."))
   n12 = float * float(
    A:int=Get l_lastCount() -> (l_lastCount=0)
    B=float - float(A=1, B=Get PercentDiffRequiredBetweenLevels() -> (PercentDiffRequiredBetweenLevels:float))
   ) {is=title K2Node_PromotableOperator}
   Branch(
    execute=@n10.else
    Condition="integer < integer"(A=Ceil(A=@n12) {is=title K2Node_CallFunction}, B=Get l_lodVertCount() -> (l_lodVertCount=0))
   )
   n13 = Get l_lodIndex() -> (l_lodIndex=0)
   n14 = float / float(A:int=Get l_lodVertCount() -> (l_lodVertCount=0), B:int=Get l_lastCount() -> (l_lastCount=0))
   n15 = "To Text (Float)"(
    Value=float * float(A:double=float - float(A=1, B=@n14), B=100)
    MinimumIntegralDigits=1
    MaximumIntegralDigits=324
    MinimumFractionalDigits=0
    MaximumFractionalDigits=1
   )
   n16 = "To Text (Float)"(
    Value=float * float(A=Get PercentDiffRequiredBetweenLevels(), B=100)
    MinimumIntegralDigits=1
    MaximumIntegralDigits=324
    MinimumFractionalDigits=0
    MaximumFractionalDigits=0
   )
   > AddError(NewError=Append!(
     A="LOD "
     B="To String (Integer)"(InInt=@n13)
     C=" is only a "
     D:String="To String (Text)"(InText=@n15)
     E:String="% vertex reduction from the previous LOD and we want at least "
     F:String="To String (Text)"(InText=@n16)
     G:String=%.
    ) -> (ReturnValue))
   n17 = Get l_lastCount() -> (l_lastCount=0)
   n18 = Get l_lodVertCount() -> (l_lodVertCount=0)
   > Set l_trackedPer(l_trackedPer:double="Max (Float)"(A=float / float(A=@n18, B=@n17), B=Get l_trackedPer() -> (l_trackedPer:double))) -> (Output_Get:double)
   Set l_lastCount(execute=@n7.then_1, l_lastCount:int=Get l_lodVertCount() -> (l_lodVertCount=0)) -> (Output_Get:int=0)
   n19 = "Lerp (LinearColor)"(
    A=(R=0.3274,G=0.4167,B=0,A=1)
    B=(R=0.2865,G=0.0076,B=0,A=1)
    Alpha=NormalizeToRange!(Value:double=Get l_trackedPer(), RangeMin:double=Get PercentDiffRequiredBetweenLevels(), RangeMax:double=1) -> (ReturnValue:double)
   )
   n20 = "Equal (Float)"(A=Get l_trackedPer())
   Return Node!(
    execute=@n4.then_2
    Valid:bool=IsEmpty(InString=GetCurrentErrorString!() -> (CurrentErrorString:String))
    ReportData:Struct GART Report Data Structure=Make Struct GART Report Data(
     DisplayTitle_20_#1=LOD
     DisplayDescShort_29_#2=MinLOD Vert Count
     DisplayColorOverride_40_#3=Select(Option 0:Linear Color Structure=@n19, Option 1:Linear Color Structure=(R=0,G=0,B=0,A=0), Index=@n20) -> (ReturnValue:Linear Color Structure)
     Severity_46_#4=NewEnumerator3
    )
   )
   Get l_lastCount() -> (l_lastCount=0)
   Get l_lodVertCount() -> (l_lodVertCount=0)
  ]}
 RunFilterLogicOnAssetData {nodes [
   n1 = RunFilterLogicOnAssetData()
   > n2 = Branch(Condition=@n1.AllowLoading)
   > n3 = Load_MinLODCheck!(execute, self, AssetData:Asset Data Structure=@n1) -> (then, Valid:bool, ReportData:Struct GART Report Data Structure) {is=title K2Node_CallFunction}
   > Return Node(Valid=@n3, ReportData=@n3.ReportData)
   Return Node(execute=@n2.else)
  ]}
]
variables [PercentDiffRequiredBetweenLevels {type=number, default=0.35}]
